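# Niagara ground effect for the CanineCronutDig emote: a glowing directional burst with smoke and sparks.
# Collision_Position scatters points in owner-oriented boxes and snaps them to the ground with downward traces, feeding Smoke_Ref and Spawn_Sparks_Ref through particle reads.
# DirectionalBurst renders meshes, while Low_Med renders decals and sprites, both with animated glow and dissolve-style material parameters.
# Front-end modifiers adjust color, depth, opacity and offsets for menu previews versus gameplay.
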NiagaraSystem /BRCosmetics/Animation/Game/MainPlayer/Emotes/CanineCronutDig/FX/NS_Emote_CanineCronutDig_Ground
dependencies {
 usedDataInterfaces [
  &NiagaraDataInterfaceCurve
  &NiagaraDataInterfaceArrayColor
  &NiagaraDataInterfaceVectorCurve
  &NiagaraDataInterfaceArrayFloat3
  &NiagaraDataInterfaceArrayFloat
  &NiagaraDataInterfaceCollisionQuery
  &NiagaraDataInterfaceDebugDraw
  &NiagaraDataInterfaceParticleRead
  &NiagaraDataInterfaceColorCurve
  &NiagaraDataInterfaceSpriteRendererInfo
  &NiagaraDataInterfaceVector2DCurve
 ]
 usedDynamicInputs [&VectorFromFloat, &InvertBool]
 usedModules [
  &SystemState
  &self:SystemSpawnScript.SystemScriptSource.SystemScriptGraph.NiagaraNodeAssignment_1.SetVariables_#1
 ]
}
emitters {
 Collision_Position {
  emitterSpawnScript [
   /SetVariables(Emitter.HoldPosition1:NiagaraPosition=&AddVectorToPosition(
     Position=<0 0 0>
     Vector=/Game/Effects/Niagara/DynamicInputs/Position/FN_OffsetValueBySystemAxis(
      OffsetDirectionVector:Vector3f=&AddVectorToPosition(
       Position=&Multiply_VectorByFloat(Vector=$Engine.Owner.SystemXAxis, Float=-5)
       Vector=&Multiply_VectorByFloat(Vector=$Engine.Owner.SystemYAxis, Float=120)
      )
      Position To Offset From:Vector3f=?
      Offset Amount:NiagaraFloat=1
     )
    )) {isSetParameters=true}
   /SetVariables(Emitter.HoldPosition2:NiagaraPosition=&AddVectorToPosition(
     Position=<0 0 0>
     Vector=&FN_OffsetValueBySystemAxis(
      OffsetDirectionVector:Vector3f=&AddVectorToPosition(
       Position=&Multiply_VectorByFloat(Vector=$Engine.Owner.SystemXAxis, Float=-25)
       Vector=&Multiply_VectorByFloat(Vector=$Engine.Owner.SystemYAxis, Float=20)
      )
      Position To Offset From:Vector3f=?
      Offset Amount:NiagaraFloat=1
     )
    )) {isSetParameters=true}
   /SetVariables(Emitter.HoldPosition3:NiagaraPosition=&AddVectorToPosition(
     Position=<0 0 0>
     Vector=&FN_OffsetValueBySystemAxis(
      OffsetDirectionVector:Vector3f=&AddVectorToPosition(
       Position=&Multiply_VectorByFloat(Vector=$Engine.Owner.SystemXAxis, Float=-15)
       Vector=&Multiply_VectorByFloat(Vector=$Engine.Owner.SystemYAxis, Float=-96)
      )
      Position To Offset From:Vector3f=?
      Offset Amount:NiagaraFloat=1
     )
    )) {isSetParameters=true}
   /SetVariables(Emitter.HoldRotation:Quat4f=$Engine.Owner.Rotation) {isSetParameters=true}
  ]
  emitterUpdateScript [
   &EmitterState()
   &SpawnBurst_Instantaneous(Spawn Count=25, Use Spawn Probability=false, UseLoopCountLimit=false)
   &SpawnBurst_Instantaneous(Spawn Count=10, Spawn Group=1, Use Spawn Probability=false, UseLoopCountLimit=false) {module=SpawnBurst_Instantaneous001}
   &SpawnBurst_Instantaneous(Spawn Count=10, Spawn Group=2, Use Spawn Probability=false, UseLoopCountLimit=false) {module=SpawnBurst_Instantaneous002}
   &SpawnBurst_Instantaneous(Spawn Count=10, Spawn Group=3, Use Spawn Probability=false, UseLoopCountLimit=false) {module=SpawnBurst_Instantaneous003}
  ]
  particleSpawnScript [
   &InitializeParticle(Color Mode=::NewEnumerator0, Lifetime=5, UsePositionOffset=false, Write Material Random=true)
   &ShapeLocation(
    Additional Quaternion Rotation=true
    Additional Yaw / Pitch / Roll=&MakeVector(
     X=&SelectFloatFromArray(
      Direct Array Index=&SpawnGroup()
      Float Selection Array=DataInterface() {propertyValues {FloatData [-88,80]}}
      Random Seed=0
     )
     Y=0
     Z=0
    )
    Box Midpoint=<1 0 0>
    Box Size=&RandomRangeVector(
     Minimum=&SelectVectorFromArray(
      Direct Array Index=&SpawnGroup()
      Vector Selection Array=DataInterface() {propertyValues {FloatData [<180 20 0>, <50 20 0>]}}
     )
     Maximum=<1 1 1>
    )
    MaskSpawnGroup=false
    Offset Mode=::NewEnumerator0
    Override Seed=false
    Rotation Angle Type:ENiagara_AngleInput=::NewEnumerator0
    Rotation Mode=::NewEnumerator3
    Rotation Quaternion=$Emitter.HoldRotation
    Shape Origin=$Emitter.HoldPosition1
    Shape Primitive=::NewEnumerator2
    Surface Only=false
    UseOwnerScale=true
   )
   &ShapeLocation(
    Additional Quaternion Rotation=true
    Additional Yaw / Pitch / Roll=&MakeVector(
     X=&SelectFloatFromArray(
      Array Sampling Mode=::NewEnumerator0
      Direct Array Index=?
      Float Selection Array=DataInterface() {propertyValues {FloatData [120,-90]}}
      Random Seed=0
     )
     Y=0
     Z=0
    )
    Box Midpoint=<1 0 0>
    Box Size=<30 0 0>
    MaskSpawnGroup=false
    Override Seed=false
    Rotation Angle Type:ENiagara_AngleInput=::NewEnumerator0
    Rotation Mode=::NewEnumerator3
    Rotation Quaternion=$Emitter.HoldRotation
    Shape Origin=$Emitter.HoldPosition2
    Shape Primitive=::NewEnumerator2
    Surface Only=false
    UseOwnerScale=true
   ) {module=ShapeLocation001}
   &ShapeLocation(
    Additional Quaternion Rotation=true
    Additional Yaw / Pitch / Roll=<-90 0 0>
    Box Midpoint=<1 0 0>
    Box Size=<50 0 0>
    MaskSpawnGroup=false
    Override Seed=false
    Rotation Angle Type:ENiagara_AngleInput=::NewEnumerator0
    Rotation Mode=::NewEnumerator3
    Rotation Quaternion=$Emitter.HoldRotation
    Shape Origin=$Emitter.HoldPosition3
    Shape Primitive=::NewEnumerator2
    Surface Only=false
    UseOwnerScale=true
   ) {module=ShapeLocation002}
   /SetVariables(Particles.Position=/Niagara/DynamicInputs/Helpers/MaskPosition(
     Mask Source:NiagaraInt32=0
     Mask Parameter:NiagaraInt32=?
     Mask:NiagaraInt32=0
     Masked Position:NiagaraPosition=$Particles.ShapeLocation.ShapePosition
     Passthrough Position:NiagaraPosition=&MaskPosition(
      Mask Source:NiagaraInt32=0
      Mask Parameter:NiagaraInt32=?
      Mask:NiagaraInt32=1
      Masked Position:NiagaraPosition=$Particles.ShapeLocation001.ShapePosition
      Passthrough Position:NiagaraPosition=&MaskPosition(
       Mask Source:NiagaraInt32=0
       Mask Parameter:NiagaraInt32=?
       Mask:NiagaraInt32=2
       Masked Position:NiagaraPosition=$Particles.ShapeLocation001.ShapePosition
       Passthrough Position:NiagaraPosition=&MaskPosition(
        Mask Source:NiagaraInt32=0
        Mask Parameter:NiagaraInt32=?
        Mask:NiagaraInt32=3
        Masked Position:NiagaraPosition=$Particles.ShapeLocation002.ShapePosition
        Passthrough Position:NiagaraPosition=<0 0 0>
       )
      )
     )
    )) {isSetParameters=true}
  ]
  particleUpdateScript [
   &ParticleState()
   &self:GridCollision(TraceDir:Vector3f=<0 0 -1>, TraceDist=300, bIsFrontEnd:NiagaraBool=$User.bIsFrontEndPreview) {module=GridCollision}
   /SetVariables(Particles.Position=$Particles.CollisionPos) {isSetParameters=true}
  ]
  rendererClasses [&NiagaraSpriteRendererProperties]
  renderers [{rendererClass=&NiagaraSpriteRendererProperties}]
 }
 DirectionalBurst {
  emitterSpawnScript [
   /SetVariables(Emitter.SpawnPosition:NiagaraPosition=$Engine.Owner.Position) {isSetParameters=true, module=SetVariables_#2}
   /SetVariables(Emitter.FRONT_END_USE_WPO:NiagaraFloat=&FN_FrontEndFloatModifier(Gameplay=1, Front End=0)) {isSetParameters=true}
   /SetVariables(Emitter.FRONT-END DEPTH MOD:NiagaraFloat=&FN_FrontEndFloatModifier(Gameplay=-0.04, Front End=-0.02)) {isSetParameters=true}
  ]
  emitterUpdateScript [
   &EmitterState(Life Cycle Mode=::NewEnumerator1, Loop Behavior=::NewEnumerator1, Loop Duration=1.65, UseLoopDelay=false)
   &SpawnBurst_Instantaneous(Use Spawn Probability=false, UseLoopCountLimit=false)
   /SetVariables(Emitter.OpacityMult:NiagaraFloat=&Lerp_Float(
     A=1
     B=1
     Alpha=&FloatFromCurve(
      FloatCurve=DataInterface() {propertyValues {Curve {keys [<at 0 0>,<at 0.8 0>,<at 1 1>]}}}
      CurveIndex=&RemapRange(
       Input Value To Remap=&DistanceBetweenPositions(A=$Emitter.SpawnPosition, B=$Engine.Owner.Position)
       Input Min=0
       Input Max=$System.DistanceMax
       Output Min=0
       Output Max=1
      )
     )
    )) {isSetParameters=true, module=SetVariables_#3}
  ]
  particleSpawnScript [
   &InitializeParticle(
    Color=&SelectLinearColorFromArray(
     Array Sampling Mode=::NewEnumerator1
     Direct Array Index=&MakeIntFromFloat(Float=&FN_FrontEndFloatModifier(Gameplay=1, Front End=0))
     Color Selection Array=DataInterface() {propertyValues {ColorData [<rgba 0.2917 0 0.0085 1>, <rgba 0.1354 0.0078 0.0115 1>]}}
     Random Seed=?
    )
    Lifetime=1.5
    Mass Max=2
    Mass Mode=::NewEnumerator3
    Mesh Scale=<4 10 5>
    Mesh Scale Mode=::NewEnumerator4
    Position Offset=<0 80 10>
    Position Offset Coordinate Space=::Local
    Sprite Rotation Angle=90
    Sprite Rotation Mode=::NewEnumerator0
    Sprite Size=<600 600>
    Sprite Size Mode=::NewEnumerator4
    UsePositionOffset=true
    Write Material Random=true
   )
   &InitialMeshOrientation(Mesh Orientation Mode=::NewEnumerator1, Rotation=<0 0 0.5>, Use Rotation Vector=true)
   &CameraOffset(Camera Offset Amount=10)
  ]
  particleUpdateScript [
   &ScaleColor(
    Scale Alpha=&FloatFromCurve(FloatCurve=DataInterface() {propertyValues {Curve {keys [<at 0 0>,<at 0.005 0>,<at 0.01 1>]}}})
    ScaleA=true
    ScaleRGB=true
   )
   /SetVariables(
    Particles.DarkGlow:LinearColor=&MultiplyLinearColor(
     Linear Color=<rgba 0.0104 0.0104 0.0104 2.5>
     Scale Factor=&MakeVector4(
      X=1
      Y=1
      Z=1
      W=&Multiply_Float(
       A=&FloatFromCurve(
        FloatCurve=DataInterface() {propertyValues {Curve {keys [<at 0 0>,<at 0.005 0>,<at 0.01 1.0147>,<at 0.9 1>,<at 1 0>]}}}
        Scale Curve=&FN_FrontEndFloatModifier(Gameplay=0.8, Front End=0.5)
       )
       B=$Emitter.OpacityMult
      )
     )
    )
    Particles.GlowColor=&MultiplyLinearColor(
     Linear Color=<rgba 1.25 0.3016 0 1>
     Scale Factor=&MakeVector4(
      X=1
      Y=1
      Z=1
      W=&Multiply_Float(
       A=&FloatFromCurve(FloatCurve=DataInterface() {propertyValues {Curve {keys [<at 0 0>,<at 0.05 0>,<at 0.2 1>,<at 0.85 1>,<at 0.98 0>]}}})
       B=$Emitter.OpacityMult
      )
     )
    )
   ) {isSetParameters=true, module=SetVariables_#4}
   &ParticleState()
   &ScaleMeshSize(Scale Factor=&VectorFromCurve(VectorCurve=DataInterface() {propertyValues {XCurve {keys [<at 0 0>,<at 0.1 1>,<at 0.8349 1>,<at 1 0.8>]}, YCurve {keys [<at 0 1>,<at 1 1>]}, ZCurve {keys [<at 0 1>,<at 1 1>]}}}))
   &DynamicMaterialParameters(
    Index 0 Param 1=&FloatFromCurve(FloatCurve=DataInterface() {propertyValues {Curve {keys [<at 0 1>,<at 0.1 0>,<at 0.9 0>,<at 1 1>]}}})
    Index 0 Param 2=&FloatFromCurve(FloatCurve=DataInterface() {propertyValues {Curve {keys [<at 0 1>,<at 0.4793 1>,<at 0.5887 1>,<at 1 1>]}}})
    Module.Index0Float0Write=true
    Module.Index0Float1Write=true
    Module.Index0Float2Write=true
    Module.Index0Float3Write=true
   )
   /SetVariables(Particles.Position=&FN_FrontEndVector3Modifier(
     Gameplay=Linked() {linkedVariable {name=Particles.Position, type=NiagaraPosition}}
     Front End=&AddVectorToPosition(Position=$Particles.Initial.Position, Vector=<0 0 -10>)
    )) {isSetParameters=true}
   &OffsetPosition(Position Offset=&FN_FrontEndVector3Modifier(Gameplay=<0 0 0>, Front End=<0 0 0>))
  ]
  rendererClasses [&NiagaraMeshRendererProperties]
  renderers [
   {rendererClass=&NiagaraMeshRendererProperties, rendererIndex=0}
   {rendererClass=&NiagaraMeshRendererProperties, rendererIndex=1}
   {rendererClass=&NiagaraMeshRendererProperties, rendererIndex=2}
  ]
 }
 Low_Med {
  emitterSpawnScript [
   /SetVariables(Emitter.SpawnPosition:NiagaraPosition=$Engine.Owner.Position) {isSetParameters=true, module=SetVariables_#2}
   /SetVariables(Emitter.FE Emissive Mod:NiagaraFloat=&FN_FrontEndFloatModifier(Gameplay=1, Front End=0.5)) {isSetParameters=true}
  ]
  emitterUpdateScript [
   &EmitterState(Life Cycle Mode=::NewEnumerator1, Loop Behavior=::NewEnumerator1, Loop Duration=1.5, UseLoopDelay=false)
   &SpawnBurst_Instantaneous(Use Spawn Probability=false, UseLoopCountLimit=false)
   /SetVariables(Emitter.OpacityMult:NiagaraFloat=&FloatFromCurve(
     FloatCurve=DataInterface() {propertyValues {Curve {keys [<at 0 1.2>,<at 0.1 0>,<at 0.9 0>,<at 1 1.2>]}}}
     CurveIndex=$Emitter.NormalizedLoopAge
    )) {isSetParameters=true, module=SetVariables_#3}
   /SetVariables(Emitter.SupportOpacity:NiagaraFloat=&FloatFromCurve(
     FloatCurve=DataInterface() {propertyValues {Curve {keys [<at 0 0>,<at 0.0742 1>,<at 0.783 1>,<at 1 0>]}}}
     CurveIndex=$Emitter.NormalizedLoopAge
     Scale Curve=0.5
    )) {isSetParameters=true}
  ]
  particleSpawnScript [
   &InitializeParticle(
    Color=<rgba 0.005 0.0015 0.0005 1>
    Lifetime=1.5
    Mass Max=2
    Mass Mode=::NewEnumerator3
    Mesh Scale=<4 10 5>
    Mesh Scale Mode=::NewEnumerator4
    Position Offset=<0 80 10>
    Position Offset Coordinate Space=::Local
    Sprite Rotation Mode=::NewEnumerator0
    Sprite Size=<250 580>
    Sprite Size Mode=::NewEnumerator4
    UsePositionOffset=true
    Write Material Random=true
   )
   &InitialMeshOrientation(Mesh Orientation Mode=::NewEnumerator1, Rotation=<0 0 0.5>, Use Rotation Vector=true)
   /Niagara/Modules/Decal/Decal_Attributes(
    Decal Alignment:Vector3f=$Engine.Owner.SystemYAxis
    Decal Facing:Vector3f=$Engine.Owner.SystemZAxis
    Decal Facing Mode=1
    Decal Rotation Mode=2
    Decal Size:Vector3f=<1500 600 250>
    Orientation=1
    Set Alignment=true
    Size=1
   )
   &SpriteFacingAndAlignment(Sprite Alignment=$Engine.Owner.SystemYAxis, Sprite Facing=<0 0 1>, WriteAlignment=true, WriteFacing=true)
   /SetVariables(Particles.VisibilityTag=&MakeIntFromFloat(Float=&FN_FrontEndFloatModifier(Gameplay=0, Front End=1))) {isSetParameters=true}
  ]
  particleUpdateScript [
   &ScaleColor(
    Scale Alpha=&FloatFromCurve(FloatCurve=DataInterface() {propertyValues {Curve {keys [<at 0 0>,<at 0.005 0>,<at 0.01 1>,<at 0.85 1>,<at 1 0>]}}})
    ScaleA=true
    ScaleRGB=true
   )
   /SetVariables(
    Particles.DarkGlow:LinearColor=&MultiplyLinearColor(
     Linear Color=<rgba 0.0104 0.0104 0.0104 3>
     Scale Factor=&MakeVector4(
      X=1
      Y=1
      Z=1
      W=&Multiply_Float(
       A=&FloatFromCurve(
        FloatCurve=DataInterface() {propertyValues {Curve {keys [<at 0 0>,<at 0.005 0>,<at 0.01 1.0147>,<at 0.9 1>,<at 1 0>]}}}
        CurveIndex=$Emitter.NormalizedLoopAge
        Scale Curve=0.8
       )
       B=$Emitter.OpacityMult
      )
     )
    )
    Particles.GlowColor=&MultiplyLinearColor(
     Linear Color=<rgba 1.25 0.3016 0 1>
     Scale Factor=&MakeVector4(
      X=1
      Y=1
      Z=1
      W=&Multiply_Float(
       A=&FloatFromCurve(FloatCurve=DataInterface() {propertyValues {Curve {keys [<at 0 0>,<at 0.05 0>,<at 0.2 1>,<at 0.85 1>,<at 0.98 0>]}}})
       B=$Emitter.OpacityMult
      )
     )
    )
   ) {isSetParameters=true, module=SetVariables_#4}
   &ParticleState()
   &ScaleSpriteSize(
    Non-Uniform Curve Sprite Scale=DataInterface() {propertyValues {
      XCurve {defaultValue=0, keys [<at 0 0>,<at 0.05 1>,<at 0.8349 1>,<at 1 0.8>]}
      YCurve {keys [{tangentMode=RCTM_User, time=0, value=1}, {tangentMode=RCTM_User, time=1, value=1}]}
     }}
    Scale Sprite Size Mode=::NewEnumerator3
   )
   &Decal_Attributes(
    Decal Size:Vector3f=&Multiply_Vector(
     A=<1500 600 250>
     B=&MakeVector(
      X=1
      Y=1
      Z=&FloatFromCurve(FloatCurve=DataInterface() {propertyValues {Curve {defaultValue=1, keys [<at 0 0>,<at 0.05 1>,<at 0.8349 1>,<at 1 0.8>]}}})
     )
    )
    Size=1
   ) {module=Decal_Attributes001}
   &ScaleMeshSize(Scale Factor=&VectorFromCurve(VectorCurve=DataInterface() {propertyValues {XCurve {keys [<at 0 0>,<at 0.05 1>,<at 0.8349 1>,<at 1 0.8>]}, YCurve {keys [<at 0 1>,<at 1 1>]}, ZCurve {keys [<at 0 1>,<at 1 1>]}}}))
   &DynamicMaterialParameters(
    Index 0 Param 1=&FloatFromCurve(FloatCurve=DataInterface() {propertyValues {Curve {keys [<at 0 1>,<at 0.1 0>,<at 0.9 0>,<at 1 1>]}}})
    Index 0 Param 2=&FloatFromCurve(FloatCurve=DataInterface() {propertyValues {Curve {keys [<at 0 1>,<at 0.4793 1>,<at 0.5887 1>,<at 1 1>]}}})
    Module.Index0Float0Write=true
    Module.Index0Float1Write=true
    Module.Index0Float2Write=true
    Module.Index0Float3Write=true
   )
   &OffsetPosition(Position Offset=&FN_FrontEndVector3Modifier(Gameplay=<0 0 10>, Front End=<0 0 -9.5>))
  ]
  rendererClasses [&NiagaraDecalRendererProperties, &NiagaraSpriteRendererProperties]
  renderers [
   {rendererClass=&NiagaraDecalRendererProperties, rendererIndex=0}
   {rendererClass=&NiagaraDecalRendererProperties, rendererIndex=1}
   {rendererClass=&NiagaraDecalRendererProperties, rendererIndex=2}
   {rendererClass=&NiagaraSpriteRendererProperties, rendererIndex=3}
   {rendererClass=&NiagaraSpriteRendererProperties, rendererIndex=4}
   {rendererClass=&NiagaraSpriteRendererProperties, rendererIndex=5}
  ]
 }
 Smoke_Ref {
  emitterUpdateScript [
   &EmitterState(Life Cycle Mode=::NewEnumerator1, Loop Behavior=::NewEnumerator1, Loop Duration=0.7, UseLoopDelay=false)
   &SpawnParticlesFromOtherEmitter(
    Attribute Reader=DataInterface() {propertyValues {EmitterBinding {bindingMode=Other, emitterName=Collision_Position}}}
    Minimum Age Spawn Threshold=0.1
    Spawn Rate=30
    Use Maximum Age=false
    Use Minimum Age=false
   )
   /SetVariables(Emitter.FRONT_END_NEARCAMERAFADE:NiagaraFloat=&FN_FrontEndFloatModifier(Gameplay=100, Front End=0)) {isSetParameters=true}
  ]
  particleSpawnScript [
   &InitializeParticle(
    Color=&FNTimeOfDayColor(
     Night Color=<rgba 0.9188 0.9791 1.2 0.2>
     Morning Color=<rgba 1.2 1.2 1.2 0.2>
     Day Color=<rgba 1.2 1.2 1.2 0.2>
     Evening Color=<rgba 1.2 1.2 1.2 0.2>
    )
    Lifetime=&UniformRangedFloat(Minimum=0.3, Maximum=0.6)
    Mass Mode=::NewEnumerator1
    Sprite Rotation Angle=&AngleConversion(Angle:NiagaraFloat=0, Angle Input:ENiagara_AngleInput=::NewEnumerator0, Angle Output:ENiagara_AngleInput=::NewEnumerator1)
    Sprite Rotation Mode=::NewEnumerator0
    Sprite Size=&Multiply_Vector2D(
     A=&Vector2DFromFloat(Value=&UniformRangedFloat(Minimum=80, Maximum=130))
     B=&Lerp_Vector2(A=<1.2 1.2>, B=<0.5 0.5>, Alpha=$Emitter.NormalizedLoopAge)
    )
    Sprite Size Mode=::NewEnumerator4
    UsePositionOffset=true
    Write Material Random=true
   ) {module=InitializeParticle001}
   &SampleParticlesFromOtherEmitter()
   /SetVariables(Particles.InitSpriteRotRate=&UniformRangedFloat(Minimum=-3, Maximum=3)) {isSetParameters=true}
   &AddVelocity(Velocity=&RandomRangeVector(Minimum=<25 25 800>, Maximum=<-25 -25 200>))
  ]
  particleUpdateScript [
   &ParticleState()
   &ScaleSpriteSize(Scale Sprite Size Mode=::NewEnumerator0, Uniform Scale Factor=1.5) {module=ScaleSpriteSize001}
   &SpriteRotationRate(Rotation Rate=&Multiply_Float(
     A=$Particles.InitSpriteRotRate
     B=&FloatFromCurve(
      FloatCurve=DataInterface() {propertyValues {Curve {keys [
          {arriveTangent=-1.9214, interpMode=RCIM_Cubic, leaveTangent=-1.9214, tangentMode=RCTM_User, time=-0.0046, value=1.0113}
          {arriveTangent=-0.2496, interpMode=RCIM_Cubic, leaveTangent=-0.2496, tangentMode=RCTM_User, time=0.9896, value=0.16}
         ]}}}
      CurveIndex=$Emitter.Age
      Scale Curve=150
     )
    ))
   &ScaleColor(
    Linear Color Curve=DataInterface() {propertyValues {
      AlphaCurve {keys [<at 0.009 0>,<at 0.2534 0.35>,<at 1 0>]}
      BlueCurve {keys [<at 0 0>,<at 0.1119 0>,<at 0.3727 0.0578>,<at 0.5232 0.011>,<at 1 0>]}
      GreenCurve {keys [<at 0 7.6431>,<at 0.1119 0.3227>,<at 0.3727 0.0753>,<at 0.5232 0.0437>,<at 1 0>]}
      RedCurve {keys [<at 0 15>,<at 0.1119 1.2>,<at 0.3727 0.1302>,<at 0.5232 0.15>,<at 1 0>]}
     }}
    Scale Mode=::NewEnumerator2
   )
   &Drag(Drag=5, Use Linear Drag=true, Use Rotational Drag=false)
   &CurlNoiseForce(
    Bias Noise=false
    Noise Frequency=5
    Noise Strength=&FloatFromCurve(
     FloatCurve=DataInterface() {propertyValues {Curve {defaultValue=1, keys [{tangentMode=RCTM_User, time=0, value=0}, {tangentMode=RCTM_User, time=1, value=1}]}}}
     Scale Curve=3000
    )
    Pan Noise=true
    Pan Noise Field=<0.5 0.5 0.5>
   )
   &ScaleSpriteSize(Scale Factor=&Vector2DFromFloat(Value=&FloatFromCurve(FloatCurve=DataInterface() {propertyValues {Curve {keys [
          {arriveTangent=2.9298, interpMode=RCIM_Cubic, leaveTangent=0.4137, tangentMode=RCTM_User, time=0, value=0.25}
          {arriveTangent=-0.0186, interpMode=RCIM_Cubic, leaveTangent=1.2864, tangentMode=RCTM_User, time=1, value=1.2}
         ]}}})))
   &SubUVAnimation(End Frame=32)
   &SolveForcesAndVelocity(Clamp Velocity=false, Limit Acceleration=false)
  ]
  rendererClasses [&NiagaraSpriteRendererProperties]
  renderers [{rendererClass=&NiagaraSpriteRendererProperties}]
 }
 Spawn_Sparks_Ref {
  emitterSpawnScript [/SetVariables(Emitter.SpawnPosition:NiagaraPosition=$Engine.Owner.Position) {isSetParameters=true}]
  emitterUpdateScript [
   &EmitterState(Life Cycle Mode=::NewEnumerator1, Loop Behavior=::NewEnumerator1, Loop Duration=1.65, UseLoopDelay=false)
   &SpawnParticlesFromOtherEmitter(
    Attribute Reader=DataInterface() {propertyValues {EmitterBinding {bindingMode=Other, emitterName=Collision_Position}}}
    Spawn Rate=0
    Use Maximum Age=false
    Use Minimum Age=false
   )
   &SpawnBurst_Instantaneous(Spawn Count=75, Spawn Time=0.0667, Use Spawn Probability=false, UseLoopCountLimit=false)
   &SpawnBurst_Instantaneous(Spawn Count=75, Spawn Group=1, Spawn Time=0.0667, Use Spawn Probability=false, UseLoopCountLimit=false) {module=SpawnBurst_Instantaneous001}
   /SetVariables(Emitter.OpacityMult:NiagaraFloat=&Lerp_Float(
     A=1
     B=1
     Alpha=&FloatFromCurve(
      FloatCurve=DataInterface() {propertyValues {Curve {keys [<at 0 0>,<at 0.8 0>,<at 1 1>]}}}
      CurveIndex=&RemapRange(
       Input Value To Remap=&DistanceBetweenPositions(A=$Emitter.SpawnPosition, B=$Engine.Owner.Position)
       Input Min=0
       Input Max=$System.DistanceMax
       Output Min=0
       Output Max=1
      )
     )
    )) {isSetParameters=true}
  ]
  particleSpawnScript [
   &InitializeParticle(
    Color=<rgba 10 10 10 1>
    Lifetime Max=0.8
    Lifetime Min=0.2
    Lifetime Mode=::NewEnumerator1
    Mass Max=2
    Mass Mode=::NewEnumerator3
    Sprite Size Max=<3 14>
    Sprite Size Min=<2 6>
    Sprite Size Mode=::NewEnumerator2
    UsePositionOffset=false
    Write Material Random=true
   )
   &AddVelocity(
    Cone Angle=30
    Cone Axis=&SelectVectorFromArray(
     Direct Array Index=&MaskIntBySpawnGroup(Masked Int=0, Passthrough Int=1, Spawn Group Mask=0)
     Vector Selection Array=DataInterface() {propertyValues {FloatData [<-1 0 1.5>, <1 0 1.5>]}}
    )
    Distribution Along Cone Axis=0.2
    Module.UseConeFalloff=false
    Module.UseVelDistribution=true
    Override Seed=false
    Velocity Mode=::NewEnumerator2
    Velocity Speed=&RandomRangeFloat(Minimum=250, Maximum=750)
   )
   &ShapeLocation(
    MaskSpawnGroup=false
    Non Uniform Scale=<0.1 1.5 1>
    Offset=<0 100 -10>
    Offset Coordinate Space=::Local
    Offset Mode=::NewEnumerator0
    Override Seed=false
    Shape Primitive=::NewEnumerator4
    UseOwnerScale=true
   )
   &SampleParticlesFromOtherEmitter()
  ]
  particleUpdateScript [
   &ParticleState()
   &GravityForce()
   &Drag(Drag=&RandomRangeFloat(Minimum=0.8, Maximum=1.2), Use Linear Drag=true, Use Rotational Drag=false)
   &ScaleColor(
    Scale Alpha=&FloatFromCurve(FloatCurve=DataInterface() {propertyValues {Curve {keys [<at -0.0046 1.0113 RCIM_Cubic>,<at 1.0011 0.002 RCIM_Cubic>]}}})
    ScaleA=true
    ScaleRGB=true
   )
   &SolveForcesAndVelocity(Clamp Velocity=false, Limit Acceleration=false)
   &ScaleSpriteSizeBySpeed(
    Max Scale Factor=<1.2 3.3>
    Min Scale Factor=<0.75 0.3>
    Sample Scale Factor By Curve=true
    Scale Factor Curve=DataInterface() {propertyValues {Curve {keys [
        {arriveTangent=2.865, interpMode=RCIM_Cubic, time=0.1247, value=-0.0006}
        {arriveTangent=4.1522, interpMode=RCIM_Cubic, leaveTangent=4.1522, tangentMode=RCTM_User, time=0.3948, value=0.4893}
        <at 0.9958 0.9974 RCIM_Cubic>
       ]}}}
    Velocity Threshold=1500
   )
   &SubUVAnimation(
    Random Start Frame=true
    "Sprite Renderer "=DataInterface() {propertyValues {SpriteRenderer=&self:Spawn_Sparks_Ref_0.NiagaraSpriteRendererProperties_0}}
    SubUV Lookup Index Scale=0
    UseEndFrame=false
    UseStartFrame=false
   )
   /SetVariables(Particles.Position=&AddVectorToPosition(Position=$Particles.Position, Vector=$User.TranslationSpeed)) {isSetParameters=true}
   &ScaleColor(Scale Alpha=$Emitter.OpacityMult, ScaleA=true, ScaleRGB=true) {module=ScaleColor001}
  ]
  rendererClasses [&NiagaraSpriteRendererProperties]
  renderers [{rendererClass=&NiagaraSpriteRendererProperties}]
 }
}
summary {
 systemName=NS_Emote_CanineCronutDig_Ground
 emitters [
  {emitterName=DirectionalBurst, rendererClasses [&NiagaraMeshRendererProperties]}
  {emitterName=Collision_Position, rendererClasses [&NiagaraSpriteRendererProperties]}
  {emitterName=Smoke_Ref, rendererClasses [&NiagaraSpriteRendererProperties]}
  {emitterName=Spawn_Sparks_Ref, rendererClasses [&NiagaraSpriteRendererProperties]}
  {emitterName=Low_Med, rendererClasses [&NiagaraDecalRendererProperties, &NiagaraSpriteRendererProperties]}
 ]
 userVariables [
  User.bIsFrontEnd {type=NiagaraBool}
  User.bIsFrontEndPreview {type=NiagaraBool}
  User.bool {type=NiagaraBool}
  User.DustSpawnCurve {type=NiagaraFloat, defaultValue=1, description=""}
  User.Particle_BG_Color {type=LinearColor, description="", defaultValue {struct=LinearColor, value=<rgba 1 1 1 1>}}
  User.SkeletalMesh {
   type=NiagaraDataInterfaceSkeletalMesh
   defaultValue {
    struct=NiagaraExt_VariableValue_DataInterface
    value {dataInterface=&self:NiagaraDataInterfaceSkeletalMesh_0}
   }
  }
  User.TranslationSpeed {type=Vector3f, description="", defaultValue {struct=Vector3f, value=<0 0 0>}}
 ]
}
